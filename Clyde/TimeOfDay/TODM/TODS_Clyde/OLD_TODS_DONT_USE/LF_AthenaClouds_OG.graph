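# Legacy light function material, kept in a don't-use folder, that casts moving cloud shadows for Athena.
# Two wisp layers and two subtracted big-cloud layers combine into a contrast-adjusted mask; low quality keeps only big clouds.
# An optional ship shadow darkens a sphere mask around SkyfireLoc, fading in with camera height.
# At Epic quality, safe-zone collection parameters drive a circular falloff that darkens light toward and beyond the storm edge.
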Material /Clyde/TimeOfDay/TODM/TODS_Clyde/OLD_TODS_DONT_USE/LF_AthenaClouds_OG
graph {
 properties {materialDomain=MD_LightFunction, pixelDepthOffsetMode=PDOM_Legacy, shadingModel=MSM_Unlit, usedShadingModels=Unlit}
 nodes [
  n1 = ScalarParameter() {defaultValue=1, parameterName=ContrastLow}
  n2 = ScalarParameter() {parameterName=ContrastHigh}
  n3 = StaticBool()
  n4 = TextureSample(UVs=MaterialFunctionCall!(
    PreviewInMatEditor=@n3
    Size=ScalarParameter() {defaultValue=1000, parameterName=WispSize01}
    Speed=ScalarParameter() {defaultValue=1.3686, parameterName=WispSpeed01}
    Resolution
   ) {materialFunction=/Game/TimeOfDay/LightFunction/MF_CloudLayerCoord}.UVs) {samplerSource=SSM_Wrap_WorldGroupSettings, samplerType=SAMPLERTYPE_Grayscale, texture=/Game/Environments/World/LightFunctions/T_spiralLF}
  n5 = ScalarParameter() {defaultValue=0.3, parameterName=WispClipping}
  n6 = TextureSample(UVs=MaterialFunctionCall(
    PreviewInMatEditor=@n3
    Size=ScalarParameter() {defaultValue=500, parameterName=WispSize02}
    Speed=ScalarParameter() {defaultValue=1, parameterName=WispSpeed02}
   ) {materialFunction=&MF_CloudLayerCoord}.UVs) {samplerSource=SSM_Wrap_WorldGroupSettings, samplerType=SAMPLERTYPE_Grayscale, texture=/Game/Environments/World/LightFunctions/T_FORT_LF}
  n7 = LinearInterpolate(
   A=ScalarParameter() {parameterName=WispLow}
   B=ScalarParameter() {defaultValue=1, parameterName=WispHigh}
   Alpha=Subtract(A=@n4.R, B=Multiply(A=@n5, B=@n6.R))
  )
  n8 = TextureSample(UVs=MaterialFunctionCall(
    PreviewInMatEditor=@n3
    Size=ScalarParameter() {defaultValue=90000, parameterName=BigCloudSize01}
    Speed=ScalarParameter() {defaultValue=0.5, parameterName=BigCloudSpeed01}
   ) {materialFunction=&MF_CloudLayerCoord}.UVs) {samplerSource=SSM_Wrap_WorldGroupSettings, samplerType=SAMPLERTYPE_LinearGrayscale, texture=/Game/Athena/Prototype/Terrain/T_CloudCoverage}
  n9 = TextureSample(UVs=MaterialFunctionCall(
    PreviewInMatEditor=@n3
    Size=ScalarParameter() {defaultValue=120000, parameterName=BigCloudSize02}
    Speed=ScalarParameter() {defaultValue=0.5, parameterName=BigCloudSpeed02}
   ) {materialFunction=&MF_CloudLayerCoord}.UVs) {samplerSource=SSM_Wrap_WorldGroupSettings, samplerType=SAMPLERTYPE_LinearGrayscale, texture=&T_CloudCoverage}
  n10 = Multiply(A=@n9.R, B=ScalarParameter() {defaultValue=2, parameterName=BigCloudClip})
  n11 = Multiply(A=Saturate(None=Subtract(A=@n8.R, B=@n10)), B=ScalarParameter() {defaultValue=1.2, parameterName=BigCloudBoost}) {constB=1.2}
  n12 = Add(A=Multiply(A=Saturate(None=@n7), B=@n8.R), B=@n11)
  n13 = QualitySwitch(Default=@n12, Low=@n11, High=@n12, Medium=@n12)
  n14 = LinearInterpolate(A=ScalarParameter() {parameterName=ShadowAmount}, Alpha=Clamp(None=LinearInterpolate(A=@n1, B=@n2, Alpha=@n13) {constA=-0.5, constB=8}))
  n15 = LinearInterpolate(Alpha=NamedRerouteUsage()) {constA=1, constB=0.2}
  n16 = ComponentMask(None=WorldPosition().XYZ) {g=true, r=true}
  n17 = ComponentMask(None=CollectionParameter() {collection=/Game/Athena/SafeZone/MPC_SafeZoneClouds, parameterName=SafeZoneLoc}) {g=true, r=true}
  n18 = CollectionParameter() {collection=&MPC_SafeZoneClouds, parameterName=StormThicknessKm-End}
  n19 = CollectionParameter() {collection=&MPC_SafeZoneClouds, parameterName=StormRadiusKm-End}
  n20 = CollectionParameter() {collection=&MPC_SafeZoneClouds, parameterName=StormThicknessKm}
  n21 = Saturate(None=CollectionParameter() {collection=&MPC_SafeZoneClouds, parameterName=CloudStartingBias})
  n22 = Add(
   A=CollectionParameter() {collection=&MPC_SafeZoneClouds, parameterName=StormRadiusKm}
   B=Multiply(A=ScalarParameter() {defaultValue=3, parameterName=StormRadius_PreStartOffset}, B=Multiply(A=@n21, B=@n21))
  )
  n23 = MaterialFunctionCall(Float2=LinearInterpolate(
    A=MaterialFunctionCall(X=@n18, Y=@n19) {materialFunction=&MakeFloat2}.Result
    B=MaterialFunctionCall(X=@n20, Y=@n22) {materialFunction=&MakeFloat2}.Result
    Alpha=CollectionParameter() {collection=&MPC_SafeZoneClouds, parameterName=SafeZoneScale}
   )) {materialFunction=&BreakOutFloat2Components}
  n24 = Divide(
   A=Subtract(
    A=Distance(A=@n16, B=@n17)
    B=MaterialFunctionCall!(Km=@n23.G) {materialFunction=/Engine/Functions/Engine_MaterialFunctions01/Units/Km-to-Cm}.Cm
   )
   B=MaterialFunctionCall(Km=@n23.R) {materialFunction=&Km-to-Cm}.Cm
  )
  n25 = ScalarParameter() {defaultValue=0.0952, parameterName=StormCircleOffset}
  n26 = OneMinus(None=Saturate(None=Subtract(A=@n24, B=@n25)))
  n27 = ScalarParameter() {defaultValue=16, parameterName=StormCirclePower}
  Multiply(
   A=StaticSwitchParameter(True=Multiply(A=@n14, B=@n15), False=@n14) {parameterName=useShipShadow}
   B=QualitySwitch(Default=Constant() {r=1}, Epic=Power(Base=@n26, Exp=@n27))
  )
  n28 = ComponentMask(None=CollectionParameter() {collection=&MPC_SafeZoneClouds, parameterName=SkyfireLoc}) {g=true, r=true}
  n29 = ComponentMask(None=WorldPosition().XYZ) {g=true, r=true}
  n30 = ScalarParameter() {defaultValue=160656.3125, parameterName=shipShadowRadius}
  n31 = ScalarParameter() {defaultValue=0.95, parameterName=shipShadowHardness}
  n32 = CameraPositionWS()
  n33 = Divide(
   A=Subtract(
    A=MaterialFunctionCall(Float3=@n32) {materialFunction=&BreakOutFloat3Components}.B
    B=ScalarParameter() {defaultValue=12000, parameterName=ShipShadowFadeHeightBias}
   )
   B=ScalarParameter() {defaultValue=7000, parameterName=ShipShadowFadeHeightScale}
  )
  NamedRerouteDeclaration(None=Multiply(A=SphereMask(A=@n28, B=@n29, Radius=@n30, Hardness=@n31), B=Saturate(None=@n33))) {name=ShipShadow, nodeColor=<rgba 0.8516 0 1 1>}
 ]
}
