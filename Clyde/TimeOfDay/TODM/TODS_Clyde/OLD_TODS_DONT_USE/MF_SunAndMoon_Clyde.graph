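# Sky material function, in a folder marked old and unused, that draws a sun or moon disc, blended by MoonActive.
# Disc UVs derive from the atmosphere light direction, with moon UVs computed per vertex via interpolators.
# The sun blends atmosphere disk luminance with a fake HDR sphere mask; the moon combines a tinted texture and halo, masked against bad mips in realtime reflections.
# It also outputs moon, halo and active-state values.
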MaterialFunction /Clyde/TimeOfDay/TODM/TODS_Clyde/OLD_TODS_DONT_USE/MF_SunAndMoon_Clyde
graph {nodes [
  n1 = NamedRerouteUsage()
  n2 = ScalarParameter() {defaultValue=0.5, parameterName=Param_1}
  n3 = VertexInterpolator(VS=Add(A=Divide(A=@n1, B=@n2)) {constB=0.5})
  n4 = Constant() {r=0.5}
  n5 = ScalarParameter() {defaultValue=0.8, parameterName=moon halo_Inner}
  n6 = ScalarParameter() {defaultValue=-0.2, parameterName=moon halo Hardness_Inner}
  Add(B=Multiply(
    A=SphereMask(A=@n3.PS, B=@n4, Radius=@n5, Hardness=@n6) {hardnessPercent=0.1}
    B=ScalarParameter() {defaultValue=1, parameterName=moon halo_Inner_Intensity}
   ))
  n7 = NamedRerouteUsage()
  n8 = NamedRerouteUsage()
  n9 = ScalarParameter() {group=Sun-Moon, parameterName=MoonActive}
  FunctionOutput(None=NamedRerouteDeclaration(None=LinearInterpolate(A=@n7, B=@n8, Alpha=@n9)) {name=SunMoon, nodeColor=<rgba 0 0.2969 1 1>})
  n10 = SphereMask(
   A=@n3.PS
   B=Constant() {r=0.5}
   Radius=ScalarParameter() {defaultValue=0.8, parameterName=moon halo}
   Hardness=ScalarParameter() {defaultValue=-0.2, parameterName=moon halo Hardness}
  ) {hardnessPercent=0.1}
  FunctionOutput(None=@n10) {outputName=MoonHaloMask}
  n11 = NamedRerouteUsage()
  n12 = ScalarParameter() {defaultValue=0.07, group=Sun-Moon, parameterName=MoonSize}
  n13 = VertexInterpolator(VS=Add(A=Divide(A=@n11, B=@n12)) {constB=0.5})
  n14 = TextureSampleParameter2D(UVs=@n13.PS) {
   group=Sun-Moon
   parameterName=MoonTexture
   samplerSource=SSM_Clamp_WorldGroupSettings
   samplerType=SAMPLERTYPE_LinearColor
   texture=/Game/TimeOfDay/Skydome/T_Moon_Artemis
  }
  n15 = Saturate(None=@n13.PS)
  n16 = Multiply(A=Ceil(None=@n15), B=OneMinus(None=Floor(None=@n15)))
  n17 = MaterialFunctionCall(Float2=@n16) {materialFunction=&BreakOutFloat2Components}
  n18 = Multiply(A=@n14.R, B=Multiply(A=@n17.R, B=@n17.R)) {desc=Multiply mask for bad Mipmapping in realtime reflections}
  n19 = Multiply(A=@n18) {constB=20}
  FunctionOutput(None=NamedRerouteDeclaration(None=Saturate(None=@n19)) {name=MoonMask, nodeColor=<rgba 0 1 0.1094 1>}) {outputName=MoonMask, sortPriority=3}
  FunctionOutput(None=@n9) {outputName=MoonActive, sortPriority=4}
  Multiply(A=@n19, B=@n16)
  n20 = VectorParameter() {group=Sun-Moon, parameterName=SunColor, defaultValue=<rgba 1 0.2982 0 1>}
  n21 = SphereMask(
   A=Constant2Vector()
   B=Divide(A=NamedRerouteUsage(), B=ScalarParameter() {defaultValue=0.04, group=Sun-Moon, parameterName=SunSize})
   Hardness=ScalarParameter() {defaultValue=0.5, group=Sun-Moon, parameterName=SunHardness}
  ) {attenuationRadius=0.5}
  n22 = Multiply(
   A=Multiply(A=ScalarParameter() {defaultValue=1000, group=Sun-Moon, parameterName=SunIntensity}, B=Multiply(A=@n20.RGB, B=@n21))
   B=NamedRerouteUsage()
  )
  NamedRerouteDeclaration(None=MaterialFunctionCall(LDR Input=Saturate(None=@n22), HDR Tint=@n22) {materialFunction=&DeriveHDRfromLDR}.RemappedRange) {name=SunSphereMask, nodeColor=<rgba 0.8281 0 1 1>}
  n23 = NamedRerouteUsage()
  n24 = ComponentMask(None=SkyAtmosphereLightDirection()) {b=true, g=true, r=true}
  NamedRerouteDeclaration(None=Step(X=DotProduct(A=@n23, B=@n24))) {name=HemisphereMask, nodeColor=<rgba 1 0 0.3047 1>}
  n25 = Multiply(A=@n24) {constB=-1}
  n26 = CrossProduct(A=@n25, B=Constant3Vector() {constant=<rgba 0 0 -1 0>})
  n27 = Normalize(VectorInput=CrossProduct(A=@n26, B=@n25))
  n28 = NamedRerouteUsage()
  n29 = Normalize(VectorInput=@n26)
  n30 = NamedRerouteUsage()
  NamedRerouteDeclaration(None=AppendVector(A=DotProduct(A=@n27, B=@n28), B=DotProduct(A=@n29, B=@n30))) {name=SunMoonUVs, nodeColor=<rgba 1 0 0.9141 1>}
  n31 = ScalarParameter() {defaultValue=1, group=Sun-Moon, parameterName=SunStrength}
  n32 = SkyAtmosphereLightDiskLuminance!(DiskAngularDiameterOverride)
  NamedRerouteDeclaration(None=LinearInterpolate(A=Multiply(A=@n31, B=@n32) {constA=1}, B=NamedRerouteUsage(), Alpha=ScalarParameter() {group=Sun-Moon, parameterName=UseFakeSun})) {name=Sun, nodeColor=<rgba 0 0.9766 1 1>}
  n33 = Multiply(A=@n10, B=VectorParameter() {parameterName=moon halo Color, defaultValue=<rgba 1 1 1 1>}.RGB)
  n34 = MaterialFunctionCall(
   UVs=@n3.PS
   Radius=ScalarParameter() {defaultValue=0.5, parameterName=Moon_G_Radius}
   Density=ScalarParameter() {parameterName=Moon_G_Density}
  ) {materialFunction=&RadialGradientExponential}
  n35 = ScalarParameter() {defaultValue=1, parameterName=Moon_Power}
  n36 = Add(
   A=Multiply(A=Multiply(A=@n33, B=@n34.RadialGradientExponential), B=ScalarParameter() {defaultValue=0.0003, parameterName=moon halo Hardness Intensity})
   B=Multiply(
    A=Power(Base=@n18, Exp=@n35) {constExponent=1.2}
    B=VectorParameter() {group=Sun-Moon, parameterName=MoonColor, defaultValue=<rgba 1 1 1 1>}.RGB
   )
  )
  n37 = ScalarParameter() {defaultValue=17, group=Sun-Moon, parameterName=MoonIntensity}
  NamedRerouteDeclaration(None=Multiply(A=NamedRerouteUsage(), B=Multiply(A=@n36, B=@n37))) {name=Moon, nodeColor=<rgba 0 1 0.8125 1>}
  n38 = WorldPosition()
  n39 = CameraPositionWS()
  NamedRerouteDeclaration(None=Normalize(VectorInput=Subtract(A=@n38.XYZ, B=@n39))) {name=CameraVectorForVertexShader, nodeColor=<rgba 0 1 0.0859 1>}
  TextureSampleParameter2D(UVs=@n3.PS) {
   group=Sun-Moon
   parameterName=MoonTexture
   samplerSource=SSM_Clamp_WorldGroupSettings
   texture=/Game/TimeOfDay/TODM_Sequence/Materials/Sky/Textures/T_Moon_Artemis_glow
  }
 ]}
