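# Unlit sky dome master material, in an outdated time-of-day folder, drawing atmosphere, sun, moon, stars and animated clouds.
# Swirl-distorted, panning cloud textures form a density fed to custom HLSL transmittance and Schlick-phase scattering, with aerial perspective composited over the Sky Atmosphere layer.
# Mobile and forward shading get a cheaper cloud path; Epic quality can substitute another cloud texture.
# Static switches add event variants, and reflection captures get brightness boosts.
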Material /Clyde/TimeOfDay/TODM/TODS_Clyde/OLD_TODS_DONT_USE/M_Apollo_SkyDome_Master_OG
graph {
 properties {bIsSky=true, bTangentSpaceNormal=false, pixelDepthOffsetMode=PDOM_Legacy, shadingModel=MSM_Unlit, usedShadingModels=Unlit}
 nodes [
  Constant2Vector() {r=-1}
  Constant() {r=14}
  Constant() {r=15}
  Constant() {r=6.5}
  Constant() {r=-0.003}
  Constant() {r=0.3}
  Constant() {r=8}
  Constant() {r=250000}
  n1 = Constant() {r=0.125}
  n2 = ScalarParameter() {defaultValue=0.05, group=Parent-MI-Settings, parameterName=SwirlPanSpeed}
  n3 = NamedRerouteUsage()
  n4 = MaterialFunctionCall!(
   SineAnimOffset
   Flow Direction=Constant2Vector() {g=0.1, r=-1}
   Time offset
   EdgeMaskSize
   UVS * size=Multiply(A=NamedRerouteUsage(), B=ScalarParameter() {defaultValue=30, group=Parent-MI-Settings, parameterName=SwirlTile}) {constB=0.5}
   SineAnimSpeed=ScalarParameter() {defaultValue=1.5, group=Parent-MI-Settings, parameterName=SwirlCurlSpeed}
   Grid Noise Tile Rep
   Rotation amount modulator=ScalarParameter() {defaultValue=0.4, group=Parent-MI-Settings, parameterName=SwirlAmount}
   un frac Time=MaterialFunctionCall(Speed=Multiply(A=@n2, B=@n3)) {materialFunction=&TimeWithSpeedVariable}.Time
  ) {desc=15 fewer instructions on low end, materialFunction=/Game/Packages/Fortress_Sky/Fortnite_SkyTwistFunction}
  n5 = ComponentMask(None=Multiply(A=@n4.Rotated UVs, B=@n4.Rotated UVs)) {desc=rotated uvs, g=true, r=true}
  n6 = Multiply(A=@n5, B=@n1) {constB=0.125}
  n7 = Multiply(A=ScalarParameter() {defaultValue=14, group=Parent-MI-Settings, parameterName=Cloud-MainCloudTile}, B=NamedRerouteUsage()) {constB=0.5}
  n8 = MaterialFunctionCall(Speed=Multiply(A=ScalarParameter() {defaultValue=0.02, group=Parent-MI-Settings, parameterName=Cloud-MainCloudSpeed}, B=NamedRerouteUsage())) {materialFunction=&TimeWithSpeedVariable}
  n9 = VertexInterpolator(VS=Multiply(A=@n1, B=Add(A=@n7, B=@n8.Time)) {constB=0.125})
  n10 = TextureSampleParameter2D(UVs=ShadingPathSwitch(Default=Add(A=@n6, B=@n9.PS), Forward=@n9.PS, Mobile=@n9.PS)) {
   parameterName=MainCloudTexture
   samplerSource=SSM_Wrap_WorldGroupSettings
   samplerType=SAMPLERTYPE_LinearColor
   texture=/Game/Packages/Fortress_Sky/SkyDome/Master/T_SkyDome_CloudShapes
  }
  NamedRerouteDeclaration(None=@n10.B) {name=Cloud-Main-Blue, nodeColor=<rgba 1 0 1 1>}
  n11 = ScalarParameter() {defaultValue=-0.003, group=Parent-MI-Settings, parameterName=Cloud-DetailCloudSpeed}
  n12 = NamedRerouteUsage()
  n13 = Add(
   A=Multiply(A=ScalarParameter() {defaultValue=6.5, group=Parent-MI-Settings, parameterName=Cloud-DetailCloudTile}, B=NamedRerouteUsage()) {constB=0.5}
   B=MaterialFunctionCall(Speed=Multiply(A=@n11, B=@n12)) {materialFunction=&TimeWithSpeedVariable}.Time
  )
  n14 = ScalarParameter() {defaultValue=0.3, group=Parent-MI-Settings, parameterName=Cloud-DetailCloudSwirl}
  n15 = TextureSampleParameter2D(UVs=Add(A=VertexInterpolator(VS=@n13).PS, B=Multiply(A=@n14, B=@n5))) {
   parameterName=CloudDetailTexture
   samplerSource=SSM_Wrap_WorldGroupSettings
   samplerType=SAMPLERTYPE_LinearColor
   texture=/Game/Packages/Fortress_Sky/SkyDome/Master/T_SkyDome_CloudShapesSmall
  }
  NamedRerouteDeclaration(None=@n15.B) {name=Cloud-Detail-Blue, nodeColor=<rgba 1 0 1 1>}
  n16 = Multiply(
   A=Power(Base=NamedRerouteUsage(), Exp=ScalarParameter() {defaultValue=4, group=CloudOpacitySettings, parameterName=MainCloud-OpacityMaskPower})
   B=ScalarParameter() {defaultValue=2, group=CloudOpacitySettings, parameterName=MainCloud-OpacityMaskMod}
  )
  n17 = ScalarParameter() {defaultValue=1, group=Cloud-ShapeSettings, parameterName=MainCloudShape-Mod}
  n18 = Power(Base=NamedRerouteUsage(), Exp=ScalarParameter() {defaultValue=1, group=Cloud-ShapeSettings, parameterName=MainCloudShape-Power})
  n19 = ScalarParameter() {defaultValue=1, group=Cloud-ShapeSettings, parameterName=DetailCloudShape-Mod}
  n20 = Power(Base=NamedRerouteUsage(), Exp=ScalarParameter() {defaultValue=1, group=Cloud-ShapeSettings, parameterName=DetailCloudShape-Power})
  n21 = MaterialFunctionCall(Transition Phase=Saturate(None=Multiply(A=@n17, B=@n18)), Height Texture=Saturate(None=Multiply(A=@n19, B=@n20))) {materialFunction=&HeightLerp}
  NamedRerouteDeclaration(None=LinearInterpolate(
    A=Saturate(None=@n16)
    B=@n21.Lerp Alpha No Contrast
    Alpha=ScalarParameter() {defaultValue=0.6, group=CloudOpacitySettings, parameterName=OpacityLerpAmount}
   )) {name=Cloud-Mask-Something, nodeColor=<rgba 1 0 1 1>}
  n22 = Clamp(
   None=SkyAtmosphereLightDiskLuminance!(DiskAngularDiameterOverride) {desc=Disk luminance w/color}
   Max=ScalarParameter() {defaultValue=100, parameterName=LightDiskMaxLuminance0}
  ) {maxDefault=100}
  n23 = VectorParameter() {parameterName=SunDiskColorMult, defaultValue=<rgba 1 1 1 1>}
  n24 = TextureCoordinate() {coordinateIndex=1}
  n25 = TextureSampleParameter2D(UVs=Multiply(A=@n24, B=ScalarParameter() {defaultValue=7, group=Stars, parameterName=StarTile})) {
   parameterName=CloudandStars
   samplerSource=SSM_Wrap_WorldGroupSettings
   samplerType=SAMPLERTYPE_LinearColor
   texture=/Game/Packages/Fortress_Sky/SkyDome/Master/T_SkyDome_StarCloudMask
  }
  n26 = ScalarParameter() {defaultValue=15, group=Stars, parameterName=StarMaskTile}
  n27 = ScalarParameter() {defaultValue=0.05, group=Stars, parameterName=StarMaskSpeed}
  n28 = TextureSampleParameter2D(UVs=Panner(
    Coordinate=Multiply(A=@n24, B=@n26)
    Time=MaterialFunctionCall(Speed=@n27) {materialFunction=&TimeWithSpeedVariable}.Time
   ) {speedX=-1}) {
   parameterName=CloudandStars
   samplerSource=SSM_Wrap_WorldGroupSettings
   samplerType=SAMPLERTYPE_LinearColor
   texture=&T_SkyDome_StarCloudMask
  }
  n29 = Multiply(A=@n25.R, B=@n28.G)
  n30 = Multiply(A=@n25.R, B=@n25.R)
  n31 = Multiply(
   A=Power(
    Base=FeatureLevelSwitch(Default=@n29, ES2=@n30, ES3_1=@n30, SM4_REMOVED=@n29) {desc=Save 4 instructions on es3 by removing Star panning}
    Exp=Constant() {r=3}
   )
   B=ScalarParameter() {group=Stars, parameterName=StarBrightness}
  )
  n32 = CollectionParameter() {
   collection=&FortniteMaterialParameters
   parameterName=SunAndMoonModelDirectionalVector
  }
  n33 = Multiply(A=ComponentMask(None=@n32) {b=true, g=true, r=true}) {constB=-1}
  n34 = CrossProduct(A=@n33, B=Constant3Vector() {constant=<rgba 0 0 -1 0>})
  n35 = Normalize(VectorInput=CrossProduct(A=@n34, B=@n33))
  n36 = WorldPosition()
  n37 = Normalize(VectorInput=@n34)
  n38 = Divide(
   A=AppendVector(A=DotProduct(A=@n35, B=@n36.XYZ), B=DotProduct(A=@n37, B=@n36.XYZ))
   B=ScalarParameter() {group=SunSettings, parameterName=SunMoon-Scale}
  )
  n39 = TextureSample(UVs=VertexInterpolator(VS=Add(A=@n38) {constB=0.5}).PS) {
   samplerSource=SSM_Clamp_WorldGroupSettings
   samplerType=SAMPLERTYPE_LinearColor
   texture=/Game/Packages/Fortress_Sky/SkyDome/Master/T_SunMoon
  }
  n40 = ScalarParameter() {defaultValue=-100, desc=This is so the moon will not blend in before the sun sets, group=SunSettings, parameterName=Moon-PreBlendOffset}
  n41 = Saturate(None=Multiply(A=@n39.R, B=@n40))
  n42 = ScalarParameter() {group=SunSettings, parameterName=Moon-Amount}
  n43 = CollectionParameter() {
   collection=&FortniteMaterialParameters
   parameterName=SunAndMoonModelWorldPosition
  }
  n44 = SphereMask(
   A=WorldPosition().XYZ
   B=ComponentMask(None=@n43) {b=true, g=true, r=true}
   Radius=ScalarParameter() {defaultValue=4208243.5, group=SunSettings, parameterName=SunMoon-AtmosphereMaskRadius}
   Hardness=Constant()
  )
  n45 = ScalarParameter() {defaultValue=50, group=SunSettings, parameterName=SunMoon-UVErrorMaskMod}
  n46 = Saturate(None=Add(A=Multiply(A=@n44, B=@n45), B=ScalarParameter() {defaultValue=-20, group=SunSettings, parameterName=SunMoon-UVErrorMaskAdd}))
  n47 = Multiply(
   A=Multiply(
    A=Multiply(A=@n41, B=@n42)
    B=VectorParameter() {group=SunSettings, parameterName=SunMoon-Color, defaultValue=<rgba 0.1321 0.2021 0.2083 1>}.RGB
   )
   B=ComponentMask(None=VertexInterpolator(VS=@n46).PS) {r=true}
  )
  n48 = Saturate(None=Multiply(A=@n47) {constB=10})
  n49 = Add(A=Multiply(A=@n31, B=OneMinus(None=@n48)), B=@n47)
  n50 = Add(
   A=Multiply(
    A=SkyAtmosphereViewLuminance!(WorldDirection) {desc="Luminance w/color\r\nBelow horizon is black"}
    B=ScalarParameter() {defaultValue=1, parameterName=AtmosphereLuminanceBrightness}
   )
   B=Add(
    A=Multiply(A=@n22, B=@n23.RGB)
    B=StaticSwitchParameter(True=@n49, False=@n31) {defaultValue=true, desc=Only use this if NONE of the material instances in the TODM use the moon., parameterName=UseMoon}
   )
  )
  n51 = MaterialFunctionCall() {materialFunction=&/MF_SunAndMoon_Clyde}
  NamedRerouteDeclaration(None=Add(A=Add(A=@n50, B=@n51.Result), B=@n31)) {name=SunMoonStarsAtmosphere, nodeColor=<rgba 1 0 1 1>}
  NamedRerouteDeclaration(None=Multiply(A=@n5, B=Constant() {r=0.4})) {name=Wispy-Coords, nodeColor=<rgba 1 0 1 1>}
  n52 = NamedRerouteUsage()
  n53 = ScalarParameter() {defaultValue=0.0007, group=Parent-MI-Settings, parameterName=MOBILE-Wispy2-Speed}
  n54 = NamedRerouteUsage()
  n55 = Add(
   A=Multiply(A=ScalarParameter() {defaultValue=8, group=Parent-MI-Settings, parameterName=MOBILE-Wispy2-Tile}, B=NamedRerouteUsage()) {constB=0.5}
   B=MaterialFunctionCall(Speed=Multiply(A=@n53, B=@n54)) {materialFunction=&TimeWithSpeedVariable}.Time
  )
  n56 = TextureSampleParameter2D(UVs=Add(A=NamedRerouteUsage(), B=VertexInterpolator(VS=@n55).PS)) {
   parameterName=CloudandStars
   samplerSource=SSM_Wrap_WorldGroupSettings
   samplerType=SAMPLERTYPE_LinearColor
   texture=&T_SkyDome_StarCloudMask
  }
  n57 = Power(Base=@n56.B) {constExponent=3}
  n58 = NamedRerouteUsage()
  n59 = NamedRerouteUsage()
  n60 = Saturate(None=LinearInterpolate(A=Multiply(A=@n57, B=@n58), Alpha=Saturate(None=@n59)) {constB=0.9, desc="Mix Cloud Base and wisps.\r\nTried a simple add and it makes the night too bright."})
  n61 = ScalarParameter() {defaultValue=4, group=MobileSettings, parameterName=MOBILE-MainCloudShape-Power}
  n62 = Subtract(A=Multiply(A=Power(Base=@n60, Exp=@n61), B=VectorParameter() {group=MobileSettings, parameterName=MOBILE-CloudsColor, defaultValue=<rgba 0 0 0 1>}.RGB)) {constB=0.01}
  n63 = Subtract(A=NamedRerouteUsage()) {constB=25000}
  n64 = Saturate(None=Multiply(A=Multiply(A=@n63) {constB=-1}) {constB=0})
  n65 = ScalarParameter() {
   defaultValue=0.2
   desc=Set this to zero to remove the white line at the horizon for special use cases only
   group=MobileSettings
   parameterName=MOBILE-HorizCloudFadeMod
  }
  n66 = Add(A=Saturate(None=Multiply(A=@n62)), B=Multiply(A=Multiply(A=@n64, B=@n64), B=@n65) {constB=0.5})
  n67 = Multiply(A=ComponentMask(None=@n52) {b=true, g=true, r=true}, B=@n66)
  n68 = Multiply(A=@n66, B=ComponentMask(None=@n52) {a=true})
  n69 = OneMinus(None=@n66)
  n70 = NamedRerouteUsage()
  NamedRerouteDeclaration(None=Add(A=Add(A=@n67, B=@n68), B=Multiply(A=@n69, B=@n70))) {name=MobileMaterial, nodeColor=<rgba 1 0 1 1>}
  NamedRerouteDeclaration(None=ScalarParameter() {defaultValue=1, group=Parent-MI-Settings, parameterName=Cloud-SpeedMaster}) {name=Cloud-SpeedMaster, nodeColor=<rgba 1 0 1 1>}
  n71 = TextureCoordinate()
  n72 = StaticSwitchParameter(
   True=AppendVector(
    A=ScalarParameter() {defaultValue=0.5, group=Parent-MI-Settings, parameterName=Cloud-TileMaster-U}
    B=ScalarParameter() {defaultValue=1.5, group=Parent-MI-Settings, parameterName=Cloud-TileMaster-V}
   )
   False=ScalarParameter() {defaultValue=1, group=Parent-MI-Settings, parameterName=Cloud-TileMaster}
  ) {group=Parent-MI-Settings, parameterName=Use2VectorForTileMaster}
  NamedRerouteDeclaration(None=Subtract(A=Multiply(A=@n71, B=@n72) {constB=0.5}, B=Constant() {r=1})) {name=Cloud-TileMaster, nodeColor=<rgba 1 0 1 1>}
  n73 = VectorParameter() {parameterName=DirectionJuniorFade, defaultValue=<rgba -0.2914 1 -0.419 1>}
  n74 = Saturate(None=DotProduct(A=VertexNormalWS(), B=Normalize(VectorInput=@n73.RGB)))
  NamedRerouteDeclaration(None=Power(Base=OneMinus(None=@n74), Exp=ScalarParameter() {defaultValue=1.3939, parameterName=JuniorFadeFalloff})) {name=JuniorFadeFalloff, nodeColor=<rgba 1 0 1 1>}
  MaterialExpressionMultiply_111 = Multiply(A=WorldPosition().XYZ, B=Constant3Vector() {constant=<rgba 0.5 0.5 0.25 0>}) {constB=0.1}
  NamedRerouteDeclaration(None=Multiply(
    A=SkyAtmosphereAerialPerspective!(World Position=@MaterialExpressionMultiply_111)
    B=ScalarParameter() {defaultValue=1, parameterName=Cloud_AerialPerspectiveFactor}
   ) {constB=0.3}) {name=CloudAerialPerspective, nodeColor=<rgba 1 0 1 1>}
  n75 = Constant() {r=0.5}
  n76 = Multiply(A=@n10.RGB, B=@n75)
  n77 = Multiply(A=@n75, B=@n15.RGB)
  n78 = Add(
   A=ScalarParameter() {defaultValue=0.2, desc=-1 to 1, group=Cloud-ShapeSettings, parameterName=DetailCloudShape-NegOne-to-One}
   B=@n21.Lerp Alpha No Contrast
  )
  n79 = ComponentMask(None=LinearInterpolate(A=Add(A=@n76, B=@n77), B=@n10.RGB, Alpha=Saturate(None=@n78))) {g=true, r=true}
  n80 = Multiply(
   A=DeriveNormalZ(InXY=ConstantBiasScale(None=@n79) {bias=-0.5, scale=2})
   B=VectorParameter() {group=Cloud-ShapeSettings, parameterName=Cloud-NormalsMod, defaultValue=<rgba 1 1 1 1>}.RGB
  )
  n81 = RotateAboutAxis(
   NormalizedRotationAxis=Constant3Vector() {constant=<rgba 0 0 1 0>}
   RotationAngle=Multiply(A=Constant() {r=-1}, B=Multiply(A=@n4.RotationAngleValues, B=@n4.RotationAngleValues)) {constB=-1}
   PivotPoint=Constant()
   Position=@n80
  ) {desc=9}
  n82 = Normalize(VectorInput=Add(A=Multiply(A=@n81) {constB=0.5}, B=@n80))
  NamedRerouteDeclaration(None=@n82) {name=Maybe-CloudTangentNormals, nodeColor=<rgba 1 0 1 1>}
  n83 = Transform(None=@n82)
  NamedRerouteDeclaration(None=ConstantBiasScale(None=ComponentMask(None=@n83) {b=true})) {name=NormalsZ-Depth, nodeColor=<rgba 1 0 1 1>}
  n84 = NamedRerouteUsage()
  n85 = NamedRerouteUsage()
  n86 = ScalarParameter() {defaultValue=1, group=CloudOpacitySettings, parameterName=Detail-LerpA-ModWispy1}
  n87 = ScalarParameter() {group=CloudOpacitySettings, parameterName=Detail-LerpB-ModWispy1}
  n88 = NamedRerouteUsage()
  n89 = ScalarParameter() {defaultValue=0.001, group=Parent-MI-Settings, parameterName=Wispy1-Speed}
  n90 = NamedRerouteUsage()
  n91 = Add(
   A=Multiply(A=ScalarParameter() {defaultValue=15, group=Parent-MI-Settings, parameterName=Wispy1-Tile}, B=NamedRerouteUsage()) {constB=0.5}
   B=MaterialFunctionCall(Speed=Multiply(A=@n89, B=@n90)) {materialFunction=&TimeWithSpeedVariable}.Time
  )
  n92 = TextureSampleParameter2D(UVs=Add(A=NamedRerouteUsage(), B=VertexInterpolator(VS=@n91).PS)) {
   parameterName=CloudandStars
   samplerSource=SSM_Wrap_WorldGroupSettings
   samplerType=SAMPLERTYPE_LinearColor
   texture=&T_SkyDome_StarCloudMask
  }
  n93 = Saturate(None=Multiply(A=LinearInterpolate(A=@n86, B=@n87, Alpha=@n88), B=@n92.B))
  n94 = NamedRerouteUsage()
  n95 = Multiply(A=Multiply(A=@n94, B=@n93), B=Add(A=@n94, B=@n93))
  n96 = Multiply(
   A=Add(A=Add(A=@n84, B=@n85), B=ScalarParameter() {defaultValue=-3, group=CloudOpacitySettings, parameterName=CloudOpacityAdd}) {constB=-0.5}
   B=Multiply(A=Saturate(None=@n95), B=ScalarParameter() {defaultValue=0.2, group=CloudOpacitySettings, parameterName=CloudOpacityMod})
  )
  n97 = NamedRerouteUsage()
  n98 = Multiply(
   A=Saturate(None=Add(A=@n96, B=@n97))
   B=ScalarParameter() {defaultValue=3, group=CloudOpacitySettings, parameterName=1-FinalCloudOpacityBoost}
  )
  n99 = Time()
  n100 = CollectionParameter() {collection=/Game/Athena/SafeZone/MPC_SafeZoneClouds, parameterName=CustomCloudsTime}
  n101 = Ceil(None=Saturate(None=@n100))
  n102 = Panner(
   Time=Multiply(A=LinearInterpolate(A=@n99, B=@n100, Alpha=@n101), B=ScalarParameter() {defaultValue=0.0023, parameterName=EpicCloudsPannerSpeed})
   Speed=Constant() {r=1}
  )
  n103 = QualitySwitch(
   Default=@n98
   Epic=StaticSwitchParameter(True=TextureSample(UVs=@n102) {texture=/Game/Packages/Fortress_Sky/SkyDome/Master/T_CloudWhisp_Pattern_04}.RGB, False=@n98) {defaultValue=true, parameterName=EpicUsesAltCloudTexture}
  )
  n104 = Add(A=NamedRerouteUsage(), B=@n103)
  MaterialExpressionReroute_64 = Reroute(None=Saturate(None=StaticSwitchParameter(True=@n104, False=@n103) {parameterName=IsBuffetEvent}))
  MaterialExpressionScalarParameter_61 = ScalarParameter() {defaultValue=1, parameterName=CloudHeight}
  MaterialExpressionScalarParameter_62 = ScalarParameter() {defaultValue=1, parameterName=ScatteringCoeff}
  MaterialExpressionScalarParameter_64 = ScalarParameter() {defaultValue=0.1, parameterName=AbsorbtionCoeff}
  n105 = Custom!(
   CloudDensity=@MaterialExpressionReroute_64
   CloudHeight=@MaterialExpressionScalarParameter_61
   ScatteringCoeff=@MaterialExpressionScalarParameter_62
   AbsorptionCoeff=@MaterialExpressionScalarParameter_64
  ) {
   code="float3 sigmaS = ScatteringCoeff * CloudDensity;\r\nfloat3 sigmaA = AbsorptionCoeff * CloudDensity;\r\nfloat3 sigmaE = sigmaS + sigmaA;\r\n\r\nfloat3 TransmittanceRGB = exp(-CloudHeight * sigmaE);\r\n\r\nfloat TransmittanceA = dot(TransmittanceRGB, float3(1.0/3.0, 1.0/3.0, 1.0/3.0));\r\n\r\nreturn saturate(TransmittanceA);"
   description=Cloud transmittance
   outputType=CMOT_Float1
   inputs [
    {
     inputName=CloudDensity
     input {expression=&self:MaterialExpressionReroute_64}
    }
    {
     inputName=CloudHeight
     input {expression=&self:MaterialExpressionScalarParameter_61}
    }
    {
     inputName=ScatteringCoeff
     input {expression=&self:MaterialExpressionScalarParameter_62}
    }
    {
     inputName=AbsorptionCoeff
     input {expression=&self:MaterialExpressionScalarParameter_64}
    }
   ]
  }
  NamedRerouteDeclaration(None=@n105) {name=CloudTransmittance, nodeColor=<rgba 1 0 1 1>}
  n106 = NamedRerouteUsage()
  MaterialExpressionNamedRerouteDeclaration_8 = NamedRerouteDeclaration(None=SkyAtmosphereLightDirection()) {name=SkyAtmosLightDir, nodeColor=<rgba 1 0 1 1>}
  n107 = Max(A=Add(A=DotProduct(A=@n106, B=@MaterialExpressionNamedRerouteDeclaration_8)) {constB=0.5})
  n108 = Saturate(None=Add(A=@MaterialExpressionReroute_64, B=@MaterialExpressionReroute_64))
  MaterialExpressionSkyAtmosphereLightIlluminance_0 = SkyAtmosphereLightIlluminance!(World Position=@MaterialExpressionMultiply_111)
  MaterialExpressionCameraPositionWS_1 = CameraPositionWS()
  n109 = Custom!(
   CloudDensity=@MaterialExpressionReroute_64
   CloudHeight=@MaterialExpressionScalarParameter_61
   SunLightDir=@MaterialExpressionNamedRerouteDeclaration_8
   SunLightIlluminance=@MaterialExpressionSkyAtmosphereLightIlluminance_0
   CameraPos=@MaterialExpressionCameraPositionWS_1
   ScatteringCoeff=@MaterialExpressionScalarParameter_62
   AbsorptionCoeff=@MaterialExpressionScalarParameter_64
   WorldPos=@MaterialExpressionMultiply_111
  ) {
   code="float3 sigmaS = ScatteringCoeff * CloudDensity;\r\nfloat3 sigmaA = AbsorptionCoeff * CloudDensity;\r\nfloat3 sigmaE = sigmaS + sigmaA;\r\n\r\nfloat3 TransmittanceRGB = exp(-CloudHeight * sigmaE);\r\n\r\nfloat phaseIn = saturate(dot(SunLightDir, -normalize(CameraPos-WorldPos)));\r\nfloat phaseUni = 1.0;\r\nfloat k = 0.8;\r\nfloat factor = 1.0 - k * phaseIn;\r\nfloat phaseSchlick = (1.0 - k*k) / (factor*factor);\r\n\r\nfloat phase = (0.3*phaseUni + 0.7*phaseSchlick) / (4.0 * PI);\r\n\r\nconst float boost = 1.0f;\r\n\r\nfloat3 S = SunLightIlluminance * phase * sigmaS * boost;\r\nfloat3 ScattLuminance = (S - S * TransmittanceRGB) / (sigmaE + 0.000001);\r\n\r\n\r\nfloat3 ScattLumFromTop = ScattLuminance * TransmittanceRGB;\r\n\r\n// TODO We need an ambient color matchign the sky mean illuminance. Otherwise the clouds remain black in area the sun light cannot reach.\r\n\r\nreturn max(0.0, ScattLumFromTop);"
   description=Cloud material
   inputs [
    {
     inputName=CloudDensity
     input {expression=&self:MaterialExpressionReroute_64}
    }
    {
     inputName=CloudHeight
     input {expression=&self:MaterialExpressionScalarParameter_61}
    }
    {
     inputName=SunLightDir
     input {expression=&self:MaterialExpressionNamedRerouteDeclaration_8}
    }
    {
     inputName=SunLightIlluminance
     input {expression=&self:MaterialExpressionSkyAtmosphereLightIlluminance_0}
    }
    {
     inputName=CameraPos
     input {expression=&self:MaterialExpressionCameraPositionWS_1}
    }
    {
     inputName=ScatteringCoeff
     input {expression=&self:MaterialExpressionScalarParameter_62}
    }
    {
     inputName=AbsorptionCoeff
     input {expression=&self:MaterialExpressionScalarParameter_64}
    }
    {
     inputName=WorldPos
     input {expression=&self:MaterialExpressionMultiply_111}
    }
   ]
  }
  NamedRerouteDeclaration(None=Multiply(A=Multiply(A=@n107, B=@n108), B=@n109)) {name=SomeKindOfLightDir, nodeColor=<rgba 1 0 1 1>}
  n110 = Rotator(Coordinate=TextureCoordinate(), Time=ScalarParameter() {defaultValue=5.6, parameterName=BuffetCloudWorldCustomTextureRotation})
  n111 = AppendVector(A=Multiply(A=Time(), B=ScalarParameter() {defaultValue=0.015, parameterName=BuffetCloudWorldCustomTexturePanningSpeedV}), B=Constant())
  n112 = ScalarParameter() {defaultValue=0.66, parameterName=BuffetCloudWorldCustomTextureScaleU}
  n113 = ScalarParameter() {defaultValue=1, parameterName=BuffetCloudWorldCustomTextureScaleV}
  n114 = TextureSample(UVs=Multiply(A=Add(A=@n110, B=@n111), B=AppendVector(A=@n112, B=@n113))) {texture=/Game/Packages/Fortress_Sky/SkyDome/Master/T_CloudWhisp_Pattern_02}
  n115 = ScalarParameter() {defaultValue=1.2, parameterName=BuffetCloudWorldCustomTexturePow}
  NamedRerouteDeclaration(None=Multiply(
    A=Power(Base=@n114.RGB, Exp=@n115)
    B=ScalarParameter() {group=CloudOpacitySettings, parameterName=BuffetCloudWorldCustomTextureOpacityBoost}
   )) {name=BuffetCloudWorldCloudLayer, nodeColor=<rgba 1 0.3516 0 1>}
  n116 = ScalarParameter() {defaultValue=1, parameterName=BrightnessBoostDuringReflectionCapture}
  n117 = ScalarParameter() {defaultValue=1.5, parameterName=FakeNightLightDuringReflectionCapture}
  n118 = NamedRerouteUsage()
  n119 = Desaturation(None=SkyAtmosphereDistantLightScatteredLuminance(), Fraction=ScalarParameter() {parameterName=Fraction})
  MaterialExpressionStaticSwitchParameter_10 = StaticSwitchParameter(True=Multiply(A=NamedRerouteUsage(), B=NamedRerouteUsage()), False=NamedRerouteUsage()) {parameterName=UseJuniorEventFade}
  n120 = OneMinus(None=@MaterialExpressionStaticSwitchParameter_10)
  n121 = Multiply(
   A=ScalarParameter() {defaultValue=1, parameterName=CloudBrightness}
   B=Add(A=NamedRerouteUsage(), B=Multiply(A=@n119, B=@n120) {constA=1, desc=Ambient x coverage}) {constB=0}
  )
  MaterialExpressionStaticSwitchParameter_9 = StaticSwitchParameter(True=Multiply(A=NamedRerouteUsage(), B=@n121), False=@n121) {parameterName=UseJuniorEventFade}
  n122 = OneMinus(None=NamedRerouteUsage())
  n123 = ComponentMask(None=WorldPosition().XYZ) {b=true}
  n124 = Saturate(None=Add(A=NamedRerouteDeclaration(None=@n123) {name=HorizonMask, nodeColor=<rgba 1 0 1 1>}) {constB=4500})
  MaterialExpressionNamedRerouteUsage_1 = NamedRerouteUsage()
  n125 = Custom!(
   AerialPerspective=@MaterialExpressionNamedRerouteUsage_1
   CloudLuminance=@MaterialExpressionStaticSwitchParameter_9
   CloudTransmittance=@MaterialExpressionStaticSwitchParameter_10
  ) {
   code="float3 Result;\r\n\r\nfloat CloudOpacity = 1.0 - CloudTransmittance;\r\n\r\nResult = \r\nCloudOpacity * AerialPerspective.rgb\r\n+\r\nCloudLuminance.rgb*AerialPerspective.a;\r\n\r\nreturn Result;"
   description=ApplyAerialPErspectiveOnClouds
   inputs [
    {
     inputName=AerialPerspective
     input {expression=&self:MaterialExpressionNamedRerouteUsage_1}
    }
    {
     inputName=CloudLuminance
     input {expression=&self:MaterialExpressionStaticSwitchParameter_9}
    }
    {
     inputName=CloudTransmittance
     input {expression=&self:MaterialExpressionStaticSwitchParameter_10}
    }
   ]
  }
  n126 = Saturate(None=Add(A=@n122, B=NamedRerouteUsage()))
  n127 = NamedRerouteUsage()
  n128 = Add(A=@n125, B=Multiply(A=StaticSwitchParameter(True=@n126, False=@n127) {parameterName=UseJuniorEventFade}, B=NamedRerouteUsage()))
  n129 = NamedRerouteUsage()
  n130 = Constant() {r=0.7}
  n131 = DotProduct(A=Multiply(A=Constant() {r=-1}, B=NamedRerouteUsage()), B=CameraVectorWS())
  n132 = Power(Base=ConstantBiasScale(None=Clamp(None=@n131)), Exp=Constant() {r=3})
  n133 = ScalarParameter() {defaultValue=0.05, parameterName=FakeNightLightRange}
  n134 = CollectionParameter() {collection=&FortniteMaterialParameters, parameterName=TimeOfDay_0-1}
  n135 = StaticBool() {value=true}
  n136 = Add(
   A=MaterialFunctionCall!(
    ThresholdValue=ScalarParameter() {defaultValue=0.9167, parameterName=FakeNightLight- Dusk Offset}
    ThresholdRange=@n133
    Alpha=@n134
    TwoSidedThreshold=@n135
   ) {materialFunction=/Engine/Functions/Engine_MaterialFunctions02/Masking/ThresholdWithRange}.Result
   B=MaterialFunctionCall(
    ThresholdValue=ScalarParameter() {defaultValue=0.125, parameterName=FakeNightLight- Dawn Offset}
    ThresholdRange=@n133
    Alpha=@n134
    TwoSidedThreshold=@n135
   ) {materialFunction=&ThresholdWithRange}.Result
  )
  n137 = NamedRerouteDeclaration(None=Multiply(A=@n136, B=@n136)) {name=FakeNightLightForCapture, nodeColor=<rgba 1 0 1 1>}
  n138 = VectorParameter() {parameterName=FakeNightLight, defaultValue=<rgba 0 0 0 1>}
  n139 = Add(
   A=StaticSwitchParameter(True=ShadingPathSwitch(Default=@n128, Forward=@n129, Mobile=@n129), False=NamedRerouteUsage()) {defaultValue=true, parameterName=UseClouds}
   B=Multiply(A=LinearInterpolate(A=@n130, Alpha=@n132), B=Multiply(A=@n137, B=@n138.RGB))
  )
  n140 = StaticSwitchParameter(
   True=Multiply(A=@MaterialExpressionStaticSwitchParameter_9, B=@n122)
   False=ShadingPathSwitch(Default=Multiply(A=@n124, B=@n139), Forward=@n139, Mobile=@n139)
  ) {parameterName=IsJuniorEventMidGroundClouds}
  ReflectionCapturePassSwitch!(Default=@n140, Reflection=Multiply(A=@n140, B=LinearInterpolate(A=@n116, B=@n117, Alpha=@n118) {constA=1, constB=2}))
  ScalarParameter() {defaultValue=5000000, group=SunDisk, parameterName=SunScale, sliderMax=1000, sliderMin=0.001}
  n141 = Constant() {r=1}
  StaticSwitchParameter(True=DepthFade(Opacity=@n141, FadeDistance=ScalarParameter() {defaultValue=5000, parameterName=DepthFadeJuniorEvent}), False=@n141) {parameterName=IsJuniorEventMidGroundClouds}
 ]
}
